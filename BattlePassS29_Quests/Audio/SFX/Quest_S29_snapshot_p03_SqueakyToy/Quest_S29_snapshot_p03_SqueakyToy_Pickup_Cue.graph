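# Sound cue for picking up the squeaky toy in a Season 29 Battle Pass snapshot quest.
# A Random node picks one of four pickup wave variations with equal weighting.
# It plays in the quests sound class with boosted volume and borrows the Hidden Star supply drop ambient loop attenuation for distance falloff.
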SoundCue /BattlePassS29_Quests/Audio/SFX/Quest_S29_snapshot_p03_SqueakyToy/Quest_S29_snapshot_p03_SqueakyToy_Pickup_Cue
root=Random_0
nodes [
 {id=Random_0, children [WavePlayer_0,WavePlayer_1,WavePlayer_2,WavePlayer_3], properties {weights [1,1,1,1]}}
 {
  id=WavePlayer_0
  properties {soundWaveAssetPtr=&/S29_Quest03_SqueakyToy_Pickup_01}
 }
 {
  id=WavePlayer_1
  properties {soundWaveAssetPtr=&/S29_Quest03_SqueakyToy_Pickup_02}
 }
 {
  id=WavePlayer_2
  properties {soundWaveAssetPtr=&/S29_Quest03_SqueakyToy_Pickup_03}
 }
 {
  id=WavePlayer_3
  properties {soundWaveAssetPtr=&/S29_Quest03_SqueakyToy_Pickup_04}
 }
]
properties {
 attenuationSettings=/Game/Athena/Sounds/SupplyDrops/Fort_HiddenStar_Ambient_Loop_Att
 maxDistance=2500
 soundClassObject=&Fort_Quests
 volumeMultiplier=4.5
}
